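# Post-process style face corrective AnimBlueprint for the ElegantLilyAnkle female head that adjusts under-chin, lip and jaw bones.
# A chain of additive bone-space Transform (Modify) Bone nodes shares one alpha, UnderChinCorrectAlpha.
# The update event computes that alpha by multiplying clamped remaps of the jawOpen curve and head-to-neck pitch difference.
# Debug logic prints positive jaw and mouth curves, and a celebration montage flag goes unused by the AnimGraph.
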AnimBlueprint /BRCosmetics/Characters/Player/Female/Medium/Heads/F_MED_ElegantLilyAnkle_Head/Meshes/F_MED_ElegantLilyAnkle_UnderChin_Transform_AnimBP
target_skeleton=/Game/Characters/Player/Common/Fortnite_Base_Head/Fortnite_Base_3L_Head_Skeleton
graphs [
 AnimGraph {nodes [
   n1 = Get UnderChinCorrectAlpha() -> (UnderChinCorrectAlpha:double)
   n2 = "Transform (Modify) Bone - Bone: FACIAL_C_UnderChin"(ComponentPose=Local To Component(LocalPose=Input Pose()), Alpha=@n1, Translation="-0.504,-1.174,-6.258", Rotation="0,0,50", Scale="1,1,1") {
    rotationMode=BMM_Additive
    translationMode=BMM_Additive
    translationSpace=BCS_BoneSpace
    alphaBoolBlend {blendOption=Linear}
    boneToModify {boneName=FACIAL_C_UnderChin}
    rotation {roll=50}
    translation=<-0.504 -1.174 -6.258>
   }
   n3 = "Transform (Modify) Bone - Bone: FACIAL_R_UnderChin"(ComponentPose=@n2, Alpha=@n1, Translation="0,-3,-2", Rotation="0,0,0", Scale="1,1,1") {
    translationMode=BMM_Additive
    translationSpace=BCS_BoneSpace
    alphaBoolBlend {blendOption=Linear}
    boneToModify {boneName=FACIAL_R_UnderChin}
    translation {y=-3, z=-2}
   }
   n4 = "Transform (Modify) Bone - Bone: FACIAL_L_UnderChin"(ComponentPose=@n3, Alpha=@n1, Translation="0,-3,-2", Rotation="0,0,0", Scale="1,1,1") {
    translationMode=BMM_Additive
    translationSpace=BCS_BoneSpace
    alphaBoolBlend {blendOption=Linear}
    boneToModify {boneName=FACIAL_L_UnderChin}
    translation {y=-3, z=-2}
   }
   n5 = "Transform (Modify) Bone - Bone: FACIAL_C_LipLower"(ComponentPose=@n4, Alpha=@n1, Translation="-0.002,0.483,-0.357", Rotation="0.221,0.303,24.239", Scale="1,1,1") {
    rotationMode=BMM_Additive
    rotationSpace=BCS_BoneSpace
    scaleMode=BMM_Additive
    scaleSpace=BCS_BoneSpace
    translationMode=BMM_Additive
    translationSpace=BCS_BoneSpace
    alphaBoolBlend {blendOption=Linear}
    boneToModify {boneName=FACIAL_C_LipLower}
    rotation=<rot 0.221 0.303 24.2389>
    translation=<-0.002 0.483 -0.357>
   }
   n6 = "Transform (Modify) Bone - Bone: FACIAL_L_LipLower"(ComponentPose=@n5, Alpha=@n1, Translation="-0.031,0.520,0.013", Rotation="-3.397,-0.603,0.170", Scale="1,1,1") {
    rotationMode=BMM_Additive
    rotationSpace=BCS_BoneSpace
    translationMode=BMM_Additive
    translationSpace=BCS_BoneSpace
    alphaBoolBlend {blendOption=Linear}
    boneToModify {boneName=FACIAL_L_LipLower}
    rotation=<rot -3.397 -0.603 0.17>
    translation=<-0.031 0.52 0.013>
   }
   n7 = "Transform (Modify) Bone - Bone: FACIAL_R_LipLower"(ComponentPose=@n6, Alpha=@n1, Translation="-0.132,0.642,-0.164", Rotation="8.066,1.470,0.521", Scale="1,1,1") {
    rotationMode=BMM_Additive
    rotationSpace=BCS_BoneSpace
    translationMode=BMM_Additive
    translationSpace=BCS_BoneSpace
    alphaBoolBlend {blendOption=Linear}
    boneToModify {boneName=FACIAL_R_LipLower}
    rotation=<rot 8.066 1.47 0.5209>
    translation=<-0.132 0.642 -0.164>
   }
   n8 = "Transform (Modify) Bone - Bone: FACIAL_C_LipUpper"(ComponentPose=@n7, Alpha=@n1, Translation="0.000,-0.120,-0.000", Rotation="0,0,0", Scale="1,1,1") {
    rotationMode=BMM_Additive
    rotationSpace=BCS_BoneSpace
    translationMode=BMM_Additive
    translationSpace=BCS_BoneSpace
    alphaBoolBlend {blendOption=Linear}
    boneToModify {boneName=FACIAL_C_LipUpper}
    translation {y=-0.12}
   }
   n9 = "Transform (Modify) Bone - Bone: FACIAL_C_LowerLipRotation"(ComponentPose=@n8, Alpha=@n1, Translation="0.013,0,-0.5", Rotation="0,0,2", Scale="1,1,1") {
    rotationMode=BMM_Additive
    rotationSpace=BCS_BoneSpace
    translationMode=BMM_Additive
    translationSpace=BCS_BoneSpace
    alphaBoolBlend {blendOption=Linear}
    boneToModify {boneName=FACIAL_C_LowerLipRotation}
    rotation {roll=2}
    translation {x=0.013, z=-0.5}
   }
   n10 = "Transform (Modify) Bone - Bone: FACIAL_C_Jaw"(ComponentPose=@n9, Alpha=@n1, Translation="0,0,-0.5", Rotation="0,0,2", Scale="1,1,1") {
    rotationMode=BMM_Additive
    rotationSpace=BCS_BoneSpace
    translationMode=BMM_Additive
    translationSpace=BCS_BoneSpace
    alphaBoolBlend {blendOption=Linear}
    boneToModify {boneName=FACIAL_C_Jaw}
    rotation {roll=2}
    translation {z=-0.5}
   }
   Output Pose(Result=Component To Local(ComponentPose=@n10)) {layerGroup=None}
  ]}
 EventGraph {nodes [
   Event BlueprintUpdateAnimation()
   n1 = GetPartSkeletalMeshComponent()
   n2 = GetCurveValue(self=GetAnimInstance(self=@n1), CurveName=CTRL_expressions_jawOpen)
   n3 = "Delta (Rotator)"!(
    A:Rotator=GetSocketRotation!(self:Scene Component Object Reference=@n1, InSocketName:Name=head) -> (ReturnValue:Rotator)
    B:Rotator=GetSocketRotation(self=@n1, InSocketName=neck_01)
   ) -> (ReturnValue_Roll:float, ReturnValue_Pitch:float, ReturnValue_Yaw:float)
   > Set UnderChinCorrectAlpha(UnderChinCorrectAlpha:double=float * float(
     A=Remap Float(Input=@n2, OldRange=(X=0,Y=0.8), NewRange=(X=0,Y=1), Clamp=true) -> (Output=~)
     B=Remap Float(Input=@n3.ReturnValue_Pitch, OldRange=(X=15,Y=40), NewRange=(X=0,Y=1), Clamp=true) -> (Output=~)
    ) {is=title K2Node_PromotableOperator}) -> (Output_Get:double)
   n4 = GetPartSkeletalMeshComponent()
   n5 = For Each Loop(Array:Array of Names=GetActiveCurveNames!(self:Anim Instance Object Reference=GetAnimInstance(self=@n4), CurveType:EAnimCurveType Enum=AttributeCurve) -> (OutNames:Array of Names)) -> (Array Element:Name)
   n6 = "To String (Name)"(InName=@n5)
   n7 = Contains(SearchIn=@n6, Substring=jaw)
   n8 = Contains(SearchIn=@n6, Substring=mouth)
   n9 = GetCurveValue(CurveName=@n5)
   > Branch(Condition=AND Boolean(A=OR Boolean(A=@n8, B=@n7), B="float > float"(A=@n9)))
   > PrintString!(
    execute
    InString=Append!(A=Split(SourceString=@n6, InStr=expressions_, SearchDir=FromStart).RightS, B="  ", C:String="To String (Float)"(InDouble=@n9)) -> (ReturnValue)
    bPrintToScreen
    bPrintToLog
    TextColor
    Duration=GetWorldDeltaSeconds!() -> (ReturnValue:double)
    Key
   ) -> (then)
   n10 = GetOwnerSkeletalMeshComponent()
   Set Celebration is Active?(Celebration is Active?:bool=Montage_IsActive!(
     self:Anim Instance Object Reference=GetAnimInstance(self=@n10)
     Montage:Anim Montage Object Reference=/BRCosmetics/Animation/Game/MainPlayer/Emotes/Emote_Celebration_Female_M
    ) -> (ReturnValue:bool)) -> (Output_Get:bool)
  ]}
]
variables [
 Celebration is Active? {type=boolean}
 UnderChinCorrectAlpha {type=number}
]
